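# Physics asset for the F_MED_Frankie female body cosmetic, set up for secondary motion of rope bones on the pants and T-shirt.
# Four four-link chains hang in left/right pairs: capsule pants ropes constrained to the pelvis, box T-shirt ropes to spine_04.
# Capsules on pelvis, spine and thighs provide body shapes.
# Swing limits are tight and twist nearly locked, with the left pants rope allowed more swing than the right.
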PhysicsAsset /BRCosmetics/Characters/Player/Female/Medium/Bodies/F_MED_Frankie/Meshes/F_MED_Frankie_Physics
bodies {
 dyn_pants_rope01_l [{shapeType=Capsule, length=1.5, radius=1, shapeName=dyn_pants_rope01_l_capsule, center=<-0.5351 0 0>, rotation=<rot -90 -0.0099 0.0099>}]
 dyn_pants_rope01_r [{shapeType=Capsule, length=1.5, radius=1, shapeName=dyn_pants_rope01_r_capsule, center=<0.3859 0 0>, rotation=<rot -90 0 0>}]
 dyn_pants_rope02_l [{shapeType=Capsule, length=1.5, radius=1, shapeName=dyn_pants_rope02_l_capsule, center=<-0.5351 0 0>, rotation=<rot -90 -0.0099 0.0099>}]
 dyn_pants_rope02_r [{shapeType=Capsule, length=1.5, radius=1, shapeName=dyn_pants_rope02_r_capsule, center=<0.3859 0 0>, rotation=<rot -90 0 0>}]
 dyn_pants_rope03_l [{shapeType=Capsule, length=1.5, radius=1, shapeName=dyn_pants_rope03_l_capsule, center=<-0.5351 0 0>, rotation=<rot -90 -0.0099 0.0099>}]
 dyn_pants_rope03_r [{shapeType=Capsule, length=1.5, radius=1, shapeName=dyn_pants_rope03_r_capsule, center=<0.3859 0 0>, rotation=<rot -90 -0.0099 0.0099>}]
 dyn_pants_rope04_l [{shapeType=Capsule, length=1.5, radius=1, shapeName=dyn_pants_rope04_l_capsule, center=<-0.5351 0 0>, rotation=<rot -90 -0.0199 0.02>}]
 dyn_pants_rope04_r [{shapeType=Capsule, length=1.5, radius=1, shapeName=dyn_pants_rope04_r_capsule, center=<0.3859 0 0>, rotation=<rot -90 -0.0199 0.02>}]
 dyn_tshirt_rope01_l1 [{shapeType=Box, extentX=2.3178, extentY=0.5002, extentZ=1.1834, center=<-0.9313 0 0>}]
 dyn_tshirt_rope01_r1 [{shapeType=Box, extentX=2.3178, extentY=0.5002, extentZ=1.2646, center=<0.9294 0 0.2337>}]
 dyn_tshirt_rope02_l [{shapeType=Box, extentX=2.3178, extentY=0.5002, extentZ=1.1834, center=<-0.9313 0 0>}]
 dyn_tshirt_rope02_r [{shapeType=Box, extentX=2.3178, extentY=0.5002, extentZ=1.2646, center=<0.9294 0 0.2337>}]
 dyn_tshirt_rope03_l [{shapeType=Box, extentX=2.3178, extentY=0.5002, extentZ=1.1834, center=<-0.9313 0 0>}]
 dyn_tshirt_rope03_r [{shapeType=Box, extentX=2.3178, extentY=0.5002, extentZ=1.2646, center=<0.9294 0 0.2337>}]
 dyn_tshirt_rope04_l [{shapeType=Box, extentX=2.3178, extentY=0.5002, extentZ=1.1834, center=<-0.9313 0 0>}]
 dyn_tshirt_rope04_r [{shapeType=Box, extentX=2.3178, extentY=0.5002, extentZ=1.2646, center=<0.9294 0 0.2337>}]
 pelvis [{shapeType=Capsule, length=9.5651, radius=8.6482, center=<1.2244 -2.0292 0.0312>, rotation=<rot 0.1149 -89.9999 0.2837>}]
 spine_03 [{shapeType=Capsule, length=9.3883, radius=4.1715, center=<0.9617 1.111 0.1716>, rotation=<rot 1.3992 -89.995 -179.9586>}]
 spine_04 [{shapeType=Capsule, length=9.1832, radius=5.8451, center=<-0.0771 1.5524 0.5137>, rotation=<rot -0.0793 -100.9346 0.8311>}]
 thigh_l [{shapeType=Capsule, length=24.843, radius=6.7998, center=<-21.9062 -1.4417 -0.1901>, rotation=<rot 0 -80.4793 -88.3032>}]
 thigh_r [{shapeType=Capsule, length=22.6431, radius=7.0783, center=<20.6232 0.6281 -0.4133>, rotation=<rot 0 -85.5926 -90.4624>}]
}
constraints [
 {bone1Name=dyn_pants_rope01_l, bone2Name=pelvis, swing1LimitDegrees=20, swing2LimitDegrees=20, twistLimitDegrees=1}
 {bone1Name=dyn_pants_rope02_l, bone2Name=dyn_pants_rope01_l, swing1LimitDegrees=20, swing2LimitDegrees=20, twistLimitDegrees=1}
 {bone1Name=dyn_pants_rope03_l, bone2Name=dyn_pants_rope02_l, swing1LimitDegrees=20, swing2LimitDegrees=20, twistLimitDegrees=1}
 {bone1Name=dyn_pants_rope04_l, bone2Name=dyn_pants_rope03_l, swing1LimitDegrees=20, swing2LimitDegrees=20, twistLimitDegrees=1}
 {bone1Name=dyn_pants_rope01_r, bone2Name=pelvis, swing1LimitDegrees=10, swing2LimitDegrees=10, twistLimitDegrees=1}
 {bone1Name=dyn_pants_rope02_r, bone2Name=dyn_pants_rope01_r, swing1LimitDegrees=10, swing2LimitDegrees=10, twistLimitDegrees=1}
 {bone1Name=dyn_pants_rope03_r, bone2Name=dyn_pants_rope02_r, swing1LimitDegrees=10, swing2LimitDegrees=10, twistLimitDegrees=1}
 {bone1Name=dyn_pants_rope04_r, bone2Name=dyn_pants_rope03_r, swing1LimitDegrees=10, swing2LimitDegrees=10, twistLimitDegrees=1}
 {bone1Name=dyn_tshirt_rope01_l1, bone2Name=spine_04, swing1LimitDegrees=15, swing2LimitDegrees=15, twistLimitDegrees=1}
 {bone1Name=dyn_tshirt_rope02_l, bone2Name=dyn_tshirt_rope01_l1, swing1LimitDegrees=15, swing2LimitDegrees=15, twistLimitDegrees=1}
 {bone1Name=dyn_tshirt_rope03_l, bone2Name=dyn_tshirt_rope02_l, swing1LimitDegrees=15, swing2LimitDegrees=15, twistLimitDegrees=1}
 {bone1Name=dyn_tshirt_rope04_l, bone2Name=dyn_tshirt_rope03_l, swing1LimitDegrees=15, swing2LimitDegrees=15, twistLimitDegrees=1}
 {bone1Name=dyn_tshirt_rope01_r1, bone2Name=spine_04, swing1LimitDegrees=15, swing2LimitDegrees=15, twistLimitDegrees=1}
 {bone1Name=dyn_tshirt_rope02_r, bone2Name=dyn_tshirt_rope01_r1, swing1LimitDegrees=15, swing2LimitDegrees=15, twistLimitDegrees=1}
 {bone1Name=dyn_tshirt_rope03_r, bone2Name=dyn_tshirt_rope02_r, swing1LimitDegrees=15, swing2LimitDegrees=15, twistLimitDegrees=1}
 {bone1Name=dyn_tshirt_rope04_r, bone2Name=dyn_tshirt_rope03_r, swing1LimitDegrees=15, swing2LimitDegrees=15, twistLimitDegrees=1}
]
