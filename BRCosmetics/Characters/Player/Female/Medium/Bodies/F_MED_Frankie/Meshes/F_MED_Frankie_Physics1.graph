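# Physics asset for the F_MED_Frankie female player cosmetic body, setting up secondary motion for rope details on the pants and T-shirt.
# Left and right four-link chains use boxes hung from the pelvis for the pants and capsules hung from spine_04 for the shirt.
# Pants links swing with almost no twist, while right shirt links have looser limits than left.
# Pelvis, upper spine and thigh capsules supply body shapes, with no constraints between them.
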PhysicsAsset /BRCosmetics/Characters/Player/Female/Medium/Bodies/F_MED_Frankie/Meshes/F_MED_Frankie_Physics1
bodies {
 dyn_pants_rope01_l [{shapeType=Box, extentX=2.4005, extentY=2.7192, extentZ=1.4873, center=<0.0312 0 0>}]
 dyn_pants_rope01_r [{shapeType=Box, extentX=2.4005, extentY=2.7192, extentZ=1.4873, center=<0.0312 0 0>}]
 dyn_pants_rope02_l [{shapeType=Box, extentX=2.4005, extentY=2.2785, extentZ=1.4873}]
 dyn_pants_rope02_r [{shapeType=Box, extentX=2.4005, extentY=2.2785, extentZ=1.4873}]
 dyn_pants_rope03_l [{shapeType=Box, extentX=2.4005, extentY=2.2785, extentZ=1.4873}]
 dyn_pants_rope03_r [{shapeType=Box, extentX=2.4005, extentY=2.2785, extentZ=1.4873}]
 dyn_pants_rope04_l [{shapeType=Box, extentX=2.4005, extentY=2.2785, extentZ=1.4873}]
 dyn_pants_rope04_r [{shapeType=Box, extentX=2.4005, extentY=2.2785, extentZ=1.4873}]
 dyn_tshirt_rope01_l1 [{shapeType=Capsule, length=1.25, radius=1.25, shapeName=dyn_tshirt_rope01_l1_capsule, center=<-0.194 0 0>, rotation=<rot -90 -0.0131 0.0129>}]
 dyn_tshirt_rope01_r1 [{
   shapeType=Capsule
   length=1.25
   radius=1.25
   shapeName=dyn_tshirt_rope01_r1_capsule
   center=<0.2378 0 0>
   rotation=<rot -90 -179.6835 179.6833>
  }]
 dyn_tshirt_rope02_l [{shapeType=Capsule, length=1.25, radius=1.25, shapeName=dyn_tshirt_rope02_l_capsule, center=<-0.194 0 0>, rotation=<rot -90 -35.2644 35.2645>}]
 dyn_tshirt_rope02_r [{
   shapeType=Capsule
   length=1.25
   radius=1.25
   shapeName=dyn_tshirt_rope02_r_capsule
   center=<0.2378 0 0>
   rotation=<rot -90 -179.6835 179.6833>
  }]
 dyn_tshirt_rope03_l [{
   shapeType=Capsule
   length=1.25
   radius=1.25
   shapeName=dyn_tshirt_rope03_l_capsule
   center=<-0.194 0 0>
   rotation=<rot -90 -179.6835 179.6833>
  }]
 dyn_tshirt_rope03_r [{shapeType=Capsule, length=1.25, radius=1.25, shapeName=dyn_tshirt_rope03_r_capsule, center=<0.2378 0 0>, rotation=<rot -90 -0.633 0.6331>}]
 dyn_tshirt_rope04_l [{
   shapeType=Capsule
   length=1.25
   radius=1.25
   shapeName=dyn_tshirt_rope04_l_capsule
   center=<-0.1993 0 -0.0606>
   rotation=<rot -80 179.9996 -180>
  }]
 dyn_tshirt_rope04_r [{
   shapeType=Capsule
   length=1.25
   radius=1.25
   shapeName=dyn_tshirt_rope04_r_capsule
   center=<0.2431 0 0.0598>
   rotation=<rot -80 179.9997 180>
  }]
 pelvis [{shapeType=Capsule, length=9.8434, radius=8.3699, center=<1.2244 -2.0292 0.0312>, rotation=<rot 0.1149 -89.9999 0.2837>}]
 spine_03 [{shapeType=Capsule, length=9.1383, radius=7.6715, center=<0.1274 0.9534 0.1759>, rotation=<rot 1.3992 -89.9908 0.0437>}]
 spine_04 [{shapeType=Capsule, length=8.4332, radius=6.5951, center=<-0.6231 2.9996 0.5271>, rotation=<rot -0.0793 -100.9363 0.8311>}]
 thigh_l [{shapeType=Capsule, length=24.843, radius=6.7998, center=<-21.9062 -1.4417 -0.1901>, rotation=<rot 0 -80.4793 -88.3032>}]
 thigh_r [{shapeType=Capsule, length=22.6431, radius=7.0783, center=<20.6232 0.6281 -0.4133>, rotation=<rot 0 -85.5926 -90.4624>}]
}
constraints [
 {bone1Name=dyn_pants_rope01_l, bone2Name=pelvis, swing1LimitDegrees=20, swing2LimitDegrees=20, twistLimitDegrees=1}
 {bone1Name=dyn_pants_rope02_l, bone2Name=dyn_pants_rope01_l, swing1LimitDegrees=20, swing2LimitDegrees=20, twistLimitDegrees=1}
 {bone1Name=dyn_pants_rope03_l, bone2Name=dyn_pants_rope02_l, swing1LimitDegrees=20, swing2LimitDegrees=20, twistLimitDegrees=1}
 {bone1Name=dyn_pants_rope04_l, bone2Name=dyn_pants_rope03_l, swing1LimitDegrees=20, swing2LimitDegrees=20, twistLimitDegrees=1}
 {bone1Name=dyn_pants_rope01_r, bone2Name=pelvis, swing1LimitDegrees=20, swing2LimitDegrees=20, twistLimitDegrees=1}
 {bone1Name=dyn_pants_rope02_r, bone2Name=dyn_pants_rope01_r, swing1LimitDegrees=20, swing2LimitDegrees=20, twistLimitDegrees=1}
 {bone1Name=dyn_pants_rope03_r, bone2Name=dyn_pants_rope02_r, swing1LimitDegrees=20, swing2LimitDegrees=20, twistLimitDegrees=1}
 {bone1Name=dyn_pants_rope04_r, bone2Name=dyn_pants_rope03_r, swing1LimitDegrees=20, swing2LimitDegrees=20, twistLimitDegrees=1}
 {bone1Name=dyn_tshirt_rope01_l1, bone2Name=spine_04, swing1LimitDegrees=15, swing2LimitDegrees=15, twistLimitDegrees=15}
 {bone1Name=dyn_tshirt_rope02_l, bone2Name=dyn_tshirt_rope01_l1, swing1LimitDegrees=15, swing2LimitDegrees=15, twistLimitDegrees=15}
 {bone1Name=dyn_tshirt_rope03_l, bone2Name=dyn_tshirt_rope02_l, swing1LimitDegrees=15, swing2LimitDegrees=15, twistLimitDegrees=15}
 {bone1Name=dyn_tshirt_rope04_l, bone2Name=dyn_tshirt_rope03_l, swing1LimitDegrees=15, swing2LimitDegrees=15, twistLimitDegrees=15}
 {bone1Name=dyn_tshirt_rope01_r1, bone2Name=spine_04, swing1LimitDegrees=25, swing2LimitDegrees=25, twistLimitDegrees=25}
 {bone1Name=dyn_tshirt_rope02_r, bone2Name=dyn_tshirt_rope01_r1, swing1LimitDegrees=25, swing2LimitDegrees=25, twistLimitDegrees=25}
 {bone1Name=dyn_tshirt_rope03_r, bone2Name=dyn_tshirt_rope02_r, swing1LimitDegrees=25, swing2LimitDegrees=25, twistLimitDegrees=25}
 {bone1Name=dyn_tshirt_rope04_r, bone2Name=dyn_tshirt_rope03_r, swing1LimitDegrees=25, swing2LimitDegrees=25, twistLimitDegrees=25}
]
