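# Physics asset for the medium female player cosmetic body Frankie, covering secondary motion of the hair.
# Eight dyn_hair chains (a–h) use box bodies, each chain jointed bone to bone and rooted on the head with tight swing and twist limits.
# Several root segments, notably b00 and c00, are held stiffer than the segments below them.
# Capsules on head, neck_01 and upper spine provide body volumes, with no constraints linking them to each other.
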PhysicsAsset /BRCosmetics/Characters/Player/Female/Medium/Bodies/F_MED_Frankie/Meshes/Parts/F_MED_Frankie_Physics
bodies {
 dyn_hair_a01 [{shapeType=Box, extentX=2.7195, extentY=0.6111, extentZ=7.5}]
 dyn_hair_a02 [{shapeType=Box, extentX=2.5207, extentY=0.6111, extentZ=6.6681}]
 dyn_hair_a03 [{shapeType=Box, extentX=1.8101, extentY=0.6111, extentZ=7.5}]
 dyn_hair_b00 [{shapeType=Box, extentX=1.9971, extentY=3.5, extentZ=2.4662, center=<0.7419 0.0085 -0.2532>, rotation=<rot 1.3449 7.5496 -0.0718>}]
 dyn_hair_b01 [{shapeType=Box, extentX=1.7045, extentY=3.4408, extentZ=5.6907, center=<0.0963 0.0226 0.0097>, rotation=<rot 4.5911 7.5749 0.6098>}]
 dyn_hair_b02 [{shapeType=Box, extentX=1.6628, extentY=3.4408, extentZ=5.6907, center=<0.0403 0 0.0001>}]
 dyn_hair_c00 [{shapeType=Box, extentX=2, extentY=3.5, extentZ=5, center=<1.0334 0 0.3903>, rotation=<rot -3.3953 0 0>}]
 dyn_hair_c01 [{shapeType=Box, extentX=1.6628, extentY=3.4408, extentZ=5.6907, center=<0.7731 0 0>}]
 dyn_hair_c02 [{shapeType=Box, extentX=1.2388, extentY=5.3872, extentZ=6.71, center=<0.2326 0 -0.9406>, rotation=<rot 28.149 0 0>}]
 dyn_hair_d00 [{shapeType=Box, extentX=2, extentY=3.5, extentZ=5, center=<1.0334 0 0.3903>, rotation=<rot -3.3953 0 0>}]
 dyn_hair_d01 [{shapeType=Box, extentX=1.6628, extentY=3.4408, extentZ=5.6907, center=<0.7731 0 0>}]
 dyn_hair_d02 [{shapeType=Box, extentX=1.2388, extentY=5.3872, extentZ=6.71, center=<0.2326 0 -0.9406>, rotation=<rot 28.149 0 0>}]
 dyn_hair_e00 [{shapeType=Box, extentX=1.5, extentY=5, extentZ=4.5, center=<0.5377 0 0.1334>, rotation=<rot -9.1273 0 0>}]
 dyn_hair_e01 [{shapeType=Box, extentX=1.6628, extentY=5.7422, extentZ=5.6907}]
 dyn_hair_e02 [{shapeType=Box, extentX=1.6628, extentY=6.052, extentZ=6.6734, center=<-0.0598 0 -0.8367>, rotation=<rot 27.579 0.0001 -0.0001>}]
 dyn_hair_f00 [{shapeType=Box, extentX=1.5, extentY=3.5, extentZ=5, center=<1.0142 0 0.145>, rotation=<rot -2.5306 0 0>}]
 dyn_hair_f01 [{shapeType=Box, extentX=1.6628, extentY=4.9408, extentZ=5.6907, rotation=<rot 0 -16.2027 0>}]
 dyn_hair_f02 [{shapeType=Box, extentX=1.6628, extentY=4.4223, extentZ=6.6907, center=<0.2864 -0.0837 -0.8711>, rotation=<rot 34.2201 -16.2954 -0.0001>}]
 dyn_hair_g01 [{shapeType=Box, extentX=1.6628, extentY=3.4408, extentZ=5.6907}]
 dyn_hair_g02 [{shapeType=Box, extentX=1.6628, extentY=3.4408, extentZ=5.6907}]
 dyn_hair_h01 [{shapeType=Box, extentX=2.8714, extentY=2.273, extentZ=7.5}]
 dyn_hair_h02 [{shapeType=Box, extentX=2.8714, extentY=2.273, extentZ=7.5}]
 head [{shapeType=Capsule, length=4.2668, radius=5.7872, center=<4.4517 2.1539 -0.0236>, rotation=<rot -0.0195 87.7529 -90.0632>}]
 neck_01 [{shapeType=Capsule, length=6.2615, radius=3.6093, center=<2.4741 0.9131 0.0431>, rotation=<rot -89.004 0.0011 -11.1715>}]
 spine_05 [{shapeType=Capsule, length=16.7511, radius=10.9989, center=<1.8161 -1.2464 0.0004>}]
}
constraints [
 {bone1Name=dyn_hair_a01, bone2Name=head, swing1LimitDegrees=5, swing2LimitDegrees=20, twistLimitDegrees=15}
 {bone1Name=dyn_hair_a02, bone2Name=dyn_hair_a01, swing1LimitDegrees=5, swing2LimitDegrees=20, twistLimitDegrees=15}
 {bone1Name=dyn_hair_a03, bone2Name=dyn_hair_a02, swing1LimitDegrees=5, swing2LimitDegrees=20, twistLimitDegrees=15}
 {bone1Name=dyn_hair_b01, bone2Name=dyn_hair_b00, swing1LimitDegrees=5, swing2LimitDegrees=20, twistLimitDegrees=5}
 {bone1Name=dyn_hair_b02, bone2Name=dyn_hair_b01, swing1LimitDegrees=5, swing2LimitDegrees=20, twistLimitDegrees=5}
 {bone1Name=dyn_hair_c01, bone2Name=dyn_hair_c00, swing1LimitDegrees=5, swing2LimitDegrees=20, twistLimitDegrees=5}
 {bone1Name=dyn_hair_c02, bone2Name=dyn_hair_c01, swing1LimitDegrees=5, swing2LimitDegrees=20, twistLimitDegrees=5}
 {bone1Name=dyn_hair_d01, bone2Name=dyn_hair_d00, swing1LimitDegrees=5, swing2LimitDegrees=20, twistLimitDegrees=5}
 {bone1Name=dyn_hair_d02, bone2Name=dyn_hair_d01, swing1LimitDegrees=5, swing2LimitDegrees=20, twistLimitDegrees=5}
 {bone1Name=dyn_hair_e01, bone2Name=dyn_hair_e00, swing1LimitDegrees=5, swing2LimitDegrees=20, twistLimitDegrees=5}
 {bone1Name=dyn_hair_e02, bone2Name=dyn_hair_e01, swing1LimitDegrees=5, swing2LimitDegrees=20, twistLimitDegrees=5}
 {bone1Name=dyn_hair_f01, bone2Name=dyn_hair_f00, swing1LimitDegrees=5, swing2LimitDegrees=20, twistLimitDegrees=5}
 {bone1Name=dyn_hair_f02, bone2Name=dyn_hair_f01, swing1LimitDegrees=5, swing2LimitDegrees=20, twistLimitDegrees=5}
 {bone1Name=dyn_hair_g01, bone2Name=head, swing1LimitDegrees=5, swing2LimitDegrees=20, twistLimitDegrees=0}
 {bone1Name=dyn_hair_g02, bone2Name=dyn_hair_g01, swing1LimitDegrees=5, swing2LimitDegrees=20, twistLimitDegrees=5}
 {bone1Name=dyn_hair_h01, bone2Name=head, swing1LimitDegrees=5, swing2LimitDegrees=20, twistLimitDegrees=15}
 {bone1Name=dyn_hair_h02, bone2Name=dyn_hair_h01, swing1LimitDegrees=5, swing2LimitDegrees=20, twistLimitDegrees=15}
 {bone1Name=dyn_hair_b00, bone2Name=head, swing1LimitDegrees=1, swing2LimitDegrees=5, twistLimitDegrees=1}
 {bone1Name=dyn_hair_c00, bone2Name=head, swing1LimitDegrees=1, swing2LimitDegrees=5, twistLimitDegrees=1}
 {bone1Name=dyn_hair_d00, bone2Name=head, swing1LimitDegrees=5, swing2LimitDegrees=10, twistLimitDegrees=5}
 {bone1Name=dyn_hair_e00, bone2Name=head, swing1LimitDegrees=5, swing2LimitDegrees=20, twistLimitDegrees=5}
 {bone1Name=dyn_hair_f00, bone2Name=head, swing1LimitDegrees=5, swing2LimitDegrees=20, twistLimitDegrees=5}
]
