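# Physics asset for the a01 hair mesh of the medium female player character Frankie, giving its dynamic hair bones collision and joint limits.
# Eight strands, a through h, are chains of box bodies rooted on a head capsule, three bones long for strand a and two for the rest.
# Each joint links a bone to its parent with tight, asymmetric swing limits, and strands a and h allow wider twist than the others.
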PhysicsAsset /BRCosmetics/Characters/Player/Female/Medium/Bodies/F_MED_Frankie/Meshes/Parts/F_MED_Frankie_hair_a01_Physics
bodies {
 dyn_hair_a01 [{shapeType=Box, extentX=4.102, extentY=1.5619, extentZ=7.5, center=<-0.029 0 0>, rotation=<rot 6.9976 0 0>}]
 dyn_hair_a02 [{shapeType=Box, extentX=3.4128, extentY=1.5619, extentZ=7.8682, center=<0.6985 0 0.2322>, rotation=<rot 12.0037 0 -0.0001>}]
 dyn_hair_a03 [{shapeType=Box, extentX=2.4688, extentY=1.5619, extentZ=7.5426, center=<0.1984 0 0.0538>, rotation=<rot 1.2236 0 -0.0001>}]
 dyn_hair_b01 [{shapeType=Box, extentX=1.6628, extentY=3.4408, extentZ=5.6907}]
 dyn_hair_b02 [{shapeType=Box, extentX=1.6628, extentY=3.4408, extentZ=5.6907}]
 dyn_hair_c01 [{shapeType=Box, extentX=1.6628, extentY=3.4408, extentZ=5.6907}]
 dyn_hair_c02 [{shapeType=Box, extentX=1.6628, extentY=3.4408, extentZ=5.6907}]
 dyn_hair_d01 [{shapeType=Box, extentX=1.6628, extentY=3.4408, extentZ=5.6907, center=<0.7731 0 0>}]
 dyn_hair_d02 [{shapeType=Box, extentX=1.6628, extentY=3.4408, extentZ=5.7186, center=<-0.1479 0 -0.4017>}]
 dyn_hair_e01 [{shapeType=Box, extentX=1.6628, extentY=3.4408, extentZ=5.6907}]
 dyn_hair_e02 [{shapeType=Box, extentX=1.6628, extentY=4.8348, extentZ=6.6644, center=<0 0 -1.3093>, rotation=<rot 7.0349 0 0>}]
 dyn_hair_f01 [{shapeType=Box, extentX=1.6628, extentY=3.4408, extentZ=5.6907, rotation=<rot 0 -16.2027 0>}]
 dyn_hair_f02 [{shapeType=Box, extentX=1.6628, extentY=3.4408, extentZ=5.6907, rotation=<rot 0 -16.2912 0>}]
 dyn_hair_g01 [{shapeType=Box, extentX=1.6628, extentY=3.4408, extentZ=5.6907}]
 dyn_hair_g02 [{shapeType=Box, extentX=1.6628, extentY=3.4408, extentZ=5.6907}]
 dyn_hair_h01 [{shapeType=Box, extentX=2.8714, extentY=2.273, extentZ=7.5}]
 dyn_hair_h02 [{shapeType=Box, extentX=2.8714, extentY=2.273, extentZ=7.5}]
 head [{shapeType=Capsule, length=6.2668, radius=5.5372, center=<4.4512 2.1287 0.4062>, rotation=<rot -0.0195 87.7529 -90.0632>}]
}
constraints [
 {bone1Name=dyn_hair_a01, bone2Name=head, swing1LimitDegrees=5, swing2LimitDegrees=20, twistLimitDegrees=15}
 {bone1Name=dyn_hair_a02, bone2Name=dyn_hair_a01, swing1LimitDegrees=5, swing2LimitDegrees=20, twistLimitDegrees=15}
 {bone1Name=dyn_hair_a03, bone2Name=dyn_hair_a02, swing1LimitDegrees=5, swing2LimitDegrees=20, twistLimitDegrees=15}
 {bone1Name=dyn_hair_b01, bone2Name=head, swing1LimitDegrees=5, swing2LimitDegrees=20, twistLimitDegrees=5}
 {bone1Name=dyn_hair_b02, bone2Name=dyn_hair_b01, swing1LimitDegrees=5, swing2LimitDegrees=20, twistLimitDegrees=5}
 {bone1Name=dyn_hair_c01, bone2Name=head, swing1LimitDegrees=5, swing2LimitDegrees=20, twistLimitDegrees=5}
 {bone1Name=dyn_hair_c02, bone2Name=dyn_hair_c01, swing1LimitDegrees=5, swing2LimitDegrees=20, twistLimitDegrees=5}
 {bone1Name=dyn_hair_d01, bone2Name=head, swing1LimitDegrees=5, swing2LimitDegrees=20, twistLimitDegrees=5}
 {bone1Name=dyn_hair_d02, bone2Name=dyn_hair_d01, swing1LimitDegrees=5, swing2LimitDegrees=20, twistLimitDegrees=5}
 {bone1Name=dyn_hair_e01, bone2Name=head, swing1LimitDegrees=5, swing2LimitDegrees=20, twistLimitDegrees=5}
 {bone1Name=dyn_hair_e02, bone2Name=dyn_hair_e01, swing1LimitDegrees=5, swing2LimitDegrees=20, twistLimitDegrees=5}
 {bone1Name=dyn_hair_f01, bone2Name=head, swing1LimitDegrees=5, swing2LimitDegrees=20, twistLimitDegrees=5}
 {bone1Name=dyn_hair_f02, bone2Name=dyn_hair_f01, swing1LimitDegrees=5, swing2LimitDegrees=20, twistLimitDegrees=5}
 {bone1Name=dyn_hair_g01, bone2Name=head, swing1LimitDegrees=5, swing2LimitDegrees=20, twistLimitDegrees=5}
 {bone1Name=dyn_hair_g02, bone2Name=dyn_hair_g01, swing1LimitDegrees=5, swing2LimitDegrees=20, twistLimitDegrees=5}
 {bone1Name=dyn_hair_h01, bone2Name=head, swing1LimitDegrees=5, swing2LimitDegrees=20, twistLimitDegrees=15}
 {bone1Name=dyn_hair_h02, bone2Name=dyn_hair_h01, swing1LimitDegrees=5, swing2LimitDegrees=20, twistLimitDegrees=15}
]
