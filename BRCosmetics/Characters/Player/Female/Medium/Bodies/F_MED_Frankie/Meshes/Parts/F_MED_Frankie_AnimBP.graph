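# Post-processing style AnimBlueprint for the F_MED_Frankie character body part, adding secondary physics motion to the pose copied from its attached parent mesh.
# Two chained RigidBody nodes simulate the body and hair_a01 physics assets in spine_05 space, with damped component motion and custom gravity, slightly upward for hair.
# Additive Transform Bone nodes then pitch eight dyn_hair bones by small fixed angles.
# The event graph holds no logic.
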AnimBlueprint /BRCosmetics/Characters/Player/Female/Medium/Bodies/F_MED_Frankie/Meshes/Parts/F_MED_Frankie_AnimBP
target_skeleton=&/F_MED_Frankie_Skeleton
graphs [
 AnimGraph {nodes [
   n1 = RigidBody(ComponentPose=Local To Component(LocalPose=Copy Pose From Mesh() {bUseAttachedParent=true})) {
    bClampLinearTranslationLimitToRefPose=true
    bOverrideWorldGravity=true
    overridePhysicsAsset=&/F_MED_Frankie_Physics
    simulationSpace=BaseBoneSpace
    alphaBoolBlend {blendOption=Linear}
    baseBoneRef {boneName=spine_05}
    componentAppliedLinearAccClamp=<1700 1700 1000>
    componentLinearAccScale=<0.35 0.45 0.45>
    componentLinearVelScale=<0.3 0.3 0.3>
    overrideWorldGravity {z=-450}
   }
   n2 = RigidBody(ComponentPose=@n1) {
    bClampLinearTranslationLimitToRefPose=true
    bOverrideWorldGravity=true
    overridePhysicsAsset=&/F_MED_Frankie_hair_a01_Physics
    simulationSpace=BaseBoneSpace
    alphaBoolBlend {blendOption=Linear}
    baseBoneRef {boneName=spine_05}
    componentAppliedLinearAccClamp=<700 700 700>
    componentLinearAccScale=<0.35 0.45 0.45>
    componentLinearVelScale=<0.3 0.3 0.3>
    overrideWorldGravity {z=50}
   }
   n3 = "Transform (Modify) Bone - Bone: dyn_hair_b00"(ComponentPose=@n2, Alpha=1, Translation="0,0,0", Rotation="9,0,0", Scale="1,1,1") {
    rotationMode=BMM_Additive
    rotationSpace=BCS_BoneSpace
    translationMode=BMM_Additive
    translationSpace=BCS_BoneSpace
    alphaBoolBlend {blendOption=Linear}
    boneToModify {boneName=dyn_hair_b00}
    rotation {pitch=9}
   }
   n4 = "Transform (Modify) Bone - Bone: dyn_hair_c00"(ComponentPose=@n3, Alpha=1, Translation="0,0,0", Rotation="9,0,0", Scale="1,1,1") {
    rotationMode=BMM_Additive
    rotationSpace=BCS_BoneSpace
    translationMode=BMM_Additive
    translationSpace=BCS_BoneSpace
    alphaBoolBlend {blendOption=Linear}
    boneToModify {boneName=dyn_hair_c00}
    rotation {pitch=9}
   }
   n5 = "Transform (Modify) Bone - Bone: dyn_hair_b01"(ComponentPose=@n4, Alpha=1, Translation="0,0,0", Rotation="1,0,0", Scale="1,1,1") {
    rotationMode=BMM_Additive
    rotationSpace=BCS_BoneSpace
    translationMode=BMM_Additive
    translationSpace=BCS_BoneSpace
    alphaBoolBlend {blendOption=Linear}
    boneToModify {boneName=dyn_hair_b01}
    rotation {pitch=1}
   }
   n6 = "Transform (Modify) Bone - Bone: dyn_hair_c01"(ComponentPose=@n5, Alpha=1, Translation="0,0,0", Rotation="1,0,0", Scale="1,1,1") {
    rotationMode=BMM_Additive
    rotationSpace=BCS_BoneSpace
    translationMode=BMM_Additive
    translationSpace=BCS_BoneSpace
    alphaBoolBlend {blendOption=Linear}
    boneToModify {boneName=dyn_hair_c01}
    rotation {pitch=1}
   }
   n7 = "Transform (Modify) Bone - Bone: dyn_hair_f00"(ComponentPose=@n6, Alpha=1, Translation="0,0,0", Rotation="9,0,0", Scale="1,1,1") {
    rotationMode=BMM_Additive
    rotationSpace=BCS_BoneSpace
    translationMode=BMM_Additive
    translationSpace=BCS_BoneSpace
    alphaBoolBlend {blendOption=Linear}
    boneToModify {boneName=dyn_hair_f00}
    rotation {pitch=9}
   }
   n8 = "Transform (Modify) Bone - Bone: dyn_hair_e00"(ComponentPose=@n7, Alpha=1, Translation="0,0,0", Rotation="9,0,0", Scale="1,1,1") {
    rotationMode=BMM_Additive
    rotationSpace=BCS_BoneSpace
    translationMode=BMM_Additive
    translationSpace=BCS_BoneSpace
    alphaBoolBlend {blendOption=Linear}
    boneToModify {boneName=dyn_hair_e00}
    rotation {pitch=9}
   }
   n9 = "Transform (Modify) Bone - Bone: dyn_hair_f01"(ComponentPose=@n8, Alpha=1, Translation="0,0,0", Rotation="2,0,0", Scale="1,1,1") {
    rotationMode=BMM_Additive
    rotationSpace=BCS_BoneSpace
    translationMode=BMM_Additive
    translationSpace=BCS_BoneSpace
    alphaBoolBlend {blendOption=Linear}
    boneToModify {boneName=dyn_hair_f01}
    rotation {pitch=2}
   }
   n10 = "Transform (Modify) Bone - Bone: dyn_hair_e01"(ComponentPose=@n9, Alpha=1, Translation="0,0,0", Rotation="2,0,0", Scale="1,1,1") {
    rotationMode=BMM_Additive
    rotationSpace=BCS_BoneSpace
    translationMode=BMM_Additive
    translationSpace=BCS_BoneSpace
    alphaBoolBlend {blendOption=Linear}
    boneToModify {boneName=dyn_hair_e01}
    rotation {pitch=2}
   }
   Output Pose(Result=Component To Local(ComponentPose=@n10)) {layerGroup=None}
  ]}
 EventGraph {nodes [Event BlueprintUpdateAnimation()]}
]
variables []
